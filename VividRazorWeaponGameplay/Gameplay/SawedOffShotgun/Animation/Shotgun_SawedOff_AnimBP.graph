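# Animation Blueprint for the sawed-off shotgun skeleton that adds secondary physics motion to its loop bone.
# A DefaultSlot montage slot over the ref pose feeds an Anim Dynamics node that swings the loop bone with a sprung, damped, angle-limited body in actor space under reduced gravity.
# A Constraint node then pins the loop bone's translation to the handle bone.
# The Event Graph contains no active logic.
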AnimBlueprint /VividRazorWeaponGameplay/Gameplay/SawedOffShotgun/Animation/Shotgun_SawedOff_AnimBP
target_skeleton=/VividRazorWeaponGameplay/Gameplay/SawedOffShotgun/Meshes/SawedOff_Shotgun_Skeleton
graphs [
 AnimGraph {nodes [
   n1 = Local Space Ref Pose()
   n2 = Anim Dynamics - Bone: loop(ComponentPose=Local To Component(LocalPose=Slot 'DefaultSlot'(Source=@n1)), Alpha=1, ExternalForce="0,0,0") {
    angularDampingOverride=1
    angularSpringConstant=2000
    bAngularSpring=true
    bOverrideAngularDamping=true
    bUseGravityOverride=true
    gravityScale=0.35
    simulationSpace=Actor
    alphaBoolBlend {blendOption=Linear}
    boundBone {boneName=loop}
    componentLinearAccScale=<1 1 1>
    componentLinearVelScale=<1 1 1>
    gravityOverride {z=-980}
    physicsBodyDefinitions [{
      boundBone {boneName=loop}
      boxExtents=<15 15 15>
      constraintSetup {angularLimitsMax=<0 90 0>, angularLimitsMin=<0 -90 0>, angularTarget=<1 0 0>}
      localJointOffset=<4 0 0>
     }]
   }
   n3 = "Constraint - loop (T)"(ComponentPose=@n2, Alpha=1, ConstraintWeights_0=1) {
    alphaBoolBlend {blendOption=Linear}
    boneToModify {boneName=loop}
    constraintSetup [{transformType=Translation, targetBone {boneName=handle}}]
    constraintWeights [1]
   }
   Output Pose(Result=Component To Local(ComponentPose=@n3))
  ]}
 EventGraph {nodes [
   TryGetPawnOwner()
   Event BlueprintUpdateAnimation()
  ]}
]
variables []
